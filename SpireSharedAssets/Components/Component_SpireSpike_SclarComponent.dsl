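; Actor component that grows its owning B_Prop_SpireSpike_ShockWave_Growable spike, cached on BeginPlay.
; ScaleSpikeToMax shuffles a copy of the tunable max-scale array, selects a value above the spike's current scale, and returns whether one was found.
; It applies that scale instantly or calls ScaleSpireSpike, which advances an alpha by easing period over duration, sets an ease-in uniform scale, and fires SpikeScalingFinished once alpha reaches one.
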(event EventBeginPlay
  (CallFunction|InitScalarComponent))

(fn InitScalarComponent
  (bind _asb_prop_spire_spike_shock_wave_growable (Utilities|Casting|CastToB_Prop_SpireSpike_ShockWave_Growable (Components|GetOwner)))
  (Variables|Internal|SetSpireSpike _asb_prop_spire_spike_shock_wave_growable))

(fn ScaleSpikeToMax
  (params (ForceInstantScale :type Boolean))
  (bind _locspikemaxscale (LocalVariables|GetLocSpikeMaxScale))
  (if (> (Utilities|Array|Length (Variables|Tunable|GetSpikeMaxScaleArray)) 0)
    (LocalVariables|SetLocSpikeStartScale (Math|Vector|GetMaxElement (Variables|/SpireSharedAssets/GameplayActors/B_Prop_SpireSpike_ShockWave_Growable.B_Prop_SpireSpike_ShockWave_Growable_C|GetSpireSpikeScale (Variables|Internal|GetSpireSpike))))
    (LocalVariables|SetLocSpikeMaxScale 0.0)
    (bind _output_get (LocalVariables|SetLocScaleArray (Variables|Tunable|GetSpikeMaxScaleArray)))
    (Utilities|Array|Shuffle _output_get)
    (LocalVariables|SetbFoundMaxScale false)
    (bind _array_element (Utilities|Array|ForEachLoopwithBreak (LocalVariables|GetLocScaleArray))
      (:LoopBody
        (if (> _array_element (LocalVariables|GetLocSpikeStartScale))
          (LocalVariables|SetbFoundMaxScale true)
          (LocalVariables|SetLocSpikeMaxScale _array_element)))
      (:Completed
        _))
    (if (LocalVariables|GetbFoundMaxScale)
      (if (Variables|GetForceInstantScale)
        (Variables|/SpireSharedAssets/GameplayActors/B_Prop_SpireSpike_ShockWave_Growable.B_Prop_SpireSpike_ShockWave_Growable_C|SetSpireSpikeScale (Math|Conversions|ToVector(Float) _locspikemaxscale) (Variables|Internal|GetSpireSpike))
        (else
          (CallFunction|ScaleSpireSpike (LocalVariables|GetLocSpikeStartScale) _locspikemaxscale)))
      (return true)
      (else
        (return false)))
    (else
      (return false))))

(fn ScaleSpireSpike
  (params (SpikeStartScale :type Float%20%28double-precision%29) (SpikeEndScale :type Float%20%28double-precision%29))
  (Variables|Internal|SetSpikeScaleAlpha 0.0)
  (bind _output_get (Variables|Internal|SetSpikeScaleAlpha (+ (Variables|Internal|GetSpikeScaleAlpha) (/ (ScalableFloat|GetValueatLevel (Variables|Tunable|GetSF_EasingPeriod)) (ScalableFloat|GetValueatLevel (Variables|Tunable|GetSF_EasingDuration))))))
  (Variables|/SpireSharedAssets/GameplayActors/B_Prop_SpireSpike_ShockWave_Growable.B_Prop_SpireSpike_ShockWave_Growable_C|SetSpireSpikeScale (Math|Conversions|ToVector(Float) (Math|Interpolation|Ease "EaseIn" _output_get (Variables|GetSpikeStartScale) (Variables|GetSpikeEndScale) 2.0)) (Variables|Internal|GetSpireSpike))
  (if (>= (Variables|Internal|GetSpikeScaleAlpha) 1.0)
    (Default|CallSpikeScalingFinished)))
